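# Physics asset for the face accessory mesh of the female medium Bubblegum player character, covering dynamic hair and a necklace with a bomb pendant.
# Capsules on head, clavicles and upper spine form body shapes, while dyn_ bones use boxes, spheres and capsules.
# Hair buns and every strand segment constrain directly to the head rather than chaining.
# Five necklace links chain from spine_05 with tiny limit angles, ending in the bomb with locked twist and swing2.
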PhysicsAsset /Game/Characters/Player/Female/Medium/Bodies/F_MED_Bubblegum/Meshes/Parts/F_MED_Bubblegum_FaceAcc_PhysicsAsset
bodies {
 clavicle_l [{shapeType=Capsule, length=7.5896, radius=3.5576, center=<5.2112 0 0.0747>, rotation=<rot 89.1512 -179.999 -179.999>}]
 clavicle_r [{shapeType=Capsule, length=7.5896, radius=3.5576, center=<-4.1209 0 -0.4474>, rotation=<rot 85.8168 -179.9998 -179.9998>}]
 dyn_L_sideStrand_01 [{shapeType=Box, extentX=0.5, extentY=2, extentZ=8, center=<4.0368 0.0001 0>, rotation=<rot 90 0 0>}]
 dyn_L_sideStrand_02 [{shapeType=Box, extentX=0.5, extentY=2, extentZ=6.5, center=<3.4071 0 0>, rotation=<rot 90 0 0>}]
 dyn_R_sideStrand_01 [{shapeType=Box, extentX=1.5, extentY=2, extentZ=8, center=<4.2167 0.0001 0>, rotation=<rot 90 0 0>}]
 dyn_bomb [{shapeType=Sphere, radius=1.9235, center=<2.2582 -1.4069 0.2337>}]
 dyn_frontStrand_01 [{shapeType=Capsule, length=5, radius=1.1, shapeName=dyn_hair_01_capsule, center=<2.6 0 0>, rotation=<rot 90 0 0>}]
 dyn_frontStrand_02 [{shapeType=Capsule, length=4, radius=0.7, shapeName=dyn_hair_02_capsule, center=<2 0 0>, rotation=<rot 90 0 0>}]
 dyn_hairbun_l [{shapeType=Sphere, radius=4.8133, center=<2.2912 -0.1616 0.1384>}]
 dyn_hairbun_r [{shapeType=Sphere, radius=4.6245, center=<-2.4473 -0.3425 0.2259>}]
 dyn_necklace [{shapeType=Box, extentX=2, extentY=12, extentZ=5.7319, center=<1.3164 0 -0.0481>, rotation=<rot 73.8439 180 180>}]
 dyn_necklace02 [{shapeType=Box, extentX=1.5713, extentY=12, extentZ=3.821, center=<1.2934 0 0.0932>, rotation=<rot 78.4762 180 180>}]
 dyn_necklace03 [{shapeType=Box, extentX=1.3881, extentY=12, extentZ=4.2895, center=<1.6546 0 0.0907>, rotation=<rot 90 0 0>}]
 dyn_necklace04 [{shapeType=Box, extentX=1.3272, extentY=12, extentZ=3.9946, center=<2 0 0>, rotation=<rot 85.3665 0 0>}]
 dyn_necklace05 [{shapeType=Box, extentX=1.8866, extentY=12, extentZ=4.1808, center=<2 0 -0.0396>, rotation=<rot 90 0 0>}]
 head [
  {shapeType=Capsule, length=3.1674, radius=6.8453, center=<0.9303 4.0771 -0.1158>, rotation=<rot 0 92.703 -89.9999>}
  {shapeType=Capsule, length=2.6674, radius=7.5953, center=<1.9182 -1.8671 0.1554>, rotation=<rot 0 92.703 -89.9999>}
  {shapeType=Capsule, length=3.1674, radius=6.8453, center=<2.029 3.9895 0.0001>, rotation=<rot -0.0002 54.3357 -89.9992>}
 ]
 spine_05 [{shapeType=Capsule, length=15.3782, radius=14.1, center=<-2.9954 -4.2699 0>}]
}
constraints [
 {bone1Name=dyn_necklace, bone2Name=spine_05, swing1LimitDegrees=0.5, swing2LimitDegrees=1, twistLimitDegrees=0.5}
 {bone1Name=dyn_necklace02, bone2Name=dyn_necklace, swing1LimitDegrees=0.5, swing2LimitDegrees=1, twistLimitDegrees=0.5}
 {bone1Name=dyn_necklace03, bone2Name=dyn_necklace02, swing1LimitDegrees=0.5, swing2LimitDegrees=1, twistLimitDegrees=0.5}
 {bone1Name=dyn_necklace04, bone2Name=dyn_necklace03, swing1LimitDegrees=0.5, swing2LimitDegrees=1, twistLimitDegrees=0.5}
 {bone1Name=dyn_necklace05, bone2Name=dyn_necklace04, swing1LimitDegrees=0.5, swing2LimitDegrees=1, twistLimitDegrees=0.5}
 {bone1Name=dyn_hairbun_l, bone2Name=head, swing1LimitDegrees=10, swing2LimitDegrees=10, twistLimitDegrees=10}
 {bone1Name=dyn_hairbun_r, bone2Name=head, swing1LimitDegrees=10, swing2LimitDegrees=10, twistLimitDegrees=10}
 {
  bone1Name=dyn_bomb
  bone2Name=dyn_necklace05
  swing1LimitDegrees=2
  swing2LimitDegrees=45
  swing2Motion=Locked
  twistLimitDegrees=45
  twistMotion=Locked
 }
 {bone1Name=dyn_frontStrand_01, bone2Name=head, swing1LimitDegrees=25, swing2LimitDegrees=25, twistLimitDegrees=45, twistMotion=Locked}
 {bone1Name=dyn_frontStrand_02, bone2Name=head, swing1LimitDegrees=25, swing2LimitDegrees=25, twistLimitDegrees=45, twistMotion=Locked}
 {bone1Name=dyn_L_sideStrand_01, bone2Name=head, swing1LimitDegrees=30, swing2LimitDegrees=30, twistLimitDegrees=3}
 {bone1Name=dyn_L_sideStrand_02, bone2Name=head, swing1LimitDegrees=30, swing2LimitDegrees=30, twistLimitDegrees=3}
 {bone1Name=dyn_R_sideStrand_01, bone2Name=head, swing1LimitDegrees=30, swing2LimitDegrees=30, twistLimitDegrees=5}
]
