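# Animation Blueprint for the Bubblegum female medium character's face accessory mesh, adding secondary physics motion.
# The output copies the attached parent's pose into a RigidBody simulation in spine_04 base-bone space, with custom gravity and scaled component motion.
# An unconnected branch holds a second simulation layered with a baked pose on necklace and hair bones.
# Its LOD-driven alpha logic is also unwired from the update event.
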AnimBlueprint /Game/Characters/Player/Female/Medium/Bodies/F_MED_Bubblegum/Meshes/Parts/F_MED_Bubblegum_FaceAcc_Skeleton_AnimBlueprint
target_skeleton=&/F_MED_Bubblegum_FaceAcc_Skeleton
graphs [
 AnimGraph {nodes [
   n1 = RigidBody(ComponentPose=Local To Component(LocalPose=Use cached pose 'CopyPose'())) {
    bClampLinearTranslationLimitToRefPose=true
    bOverrideWorldGravity=true
    lODThreshold=1
    overridePhysicsAsset=&/F_MED_Bubblegum_FaceAcc_PhysicsAsset
    simulationSpace=BaseBoneSpace
    alphaBoolBlend {blendOption=Linear}
    baseBoneRef {boneName=spine_03}
    componentAppliedLinearAccClamp=<2000 2000 2000>
    componentLinearAccScale=<0.3 0.3 0.3>
    componentLinearVelScale=<0.3 0.3 0.3>
    overrideWorldGravity {z=-980}
   }
   Layered blend per bone!(
    BasePose=Component To Local(ComponentPose=@n1)
    BlendPoses_0
    BlendPoses_1:Pose Link Structure=Sequence Player 'F_MED_Bubblegun_FaceAcc_BakedPose'() {sequence=&/F_MED_Bubblegun_FaceAcc_BakedPose}
    BlendWeights_0
    BlendWeights_1:float=Get LOD_Alpha() -> (LOD_Alpha:double)
   ) -> (Pose) {
    blendPoses [{linkId=-1, sourceLinkId=-1}, {linkId=-1, sourceLinkId=-1}]
    blendWeights [1,1]
    layerSetup [
     {branchFilters [{blendDepth=1, boneName=dyn_necklace}]}
     {branchFilters [
       {blendDepth=1, boneName=dyn_hair_01}
       {blendDepth=1, boneName=dyn_hair_02}
       {blendDepth=1, boneName=dyn_hair_03}
       {blendDepth=1, boneName=dyn_hair_04}
       {blendDepth=1, boneName=dyn_hair_05}
      ]}
    ]
   }
   n2 = RigidBody(ComponentPose=Local To Component(LocalPose=Copy Pose From Mesh() {bUseAttachedParent=true})) {
    bClampLinearTranslationLimitToRefPose=true
    bOverrideWorldGravity=true
    lODThreshold=1
    overridePhysicsAsset=&F_MED_Bubblegum_FaceAcc_PhysicsAsset
    simulationSpace=BaseBoneSpace
    alphaBoolBlend {blendOption=Linear}
    baseBoneRef {boneName=spine_04}
    componentAppliedLinearAccClamp=<2000 2000 2000>
    componentLinearAccScale=<0.4 0.4 0.4>
    componentLinearVelScale=<0.4 0.4 0.4>
    overrideWorldGravity {z=-980}
   }
   Output Pose(Result=Component To Local(ComponentPose=@n2)) {layerGroup=None}
   Save cached pose 'CopyPose'(Pose=Copy Pose From Mesh() {bUseAttachedParent=true}) {cacheName=CopyPose}
  ]}
 EventGraph {nodes [
   n1 = Branch(Condition="integer > integer"(A=Get CurrentLOD(), B=1) {is=title K2Node_CallFunction})
   > "Set <LOD2"("<LOD2":bool=true) -> (Output_Get:bool)
   > Set LOD_Alpha(LOD_Alpha:double=1) -> (Output_Get:double)
   "Set <LOD2"(execute=@n1.else, "<LOD2"=false)
   > Set LOD_Alpha(LOD_Alpha=0)
   TryGetPawnOwner()
   Event BlueprintUpdateAnimation()
  ]}
]
variables [
 "<LOD2" {type=boolean}
 LOD_Alpha {type=number}
]
